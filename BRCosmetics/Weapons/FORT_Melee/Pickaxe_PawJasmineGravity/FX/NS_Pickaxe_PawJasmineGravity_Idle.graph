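# Idle Niagara effect for the PawJasmineGravity pickaxe, showing two persistent socket-attached glows, one warm orange and one blue.
# A single Glows emitter bursts two particles, using the partition index to choose socket, color, size, camera offset and light radius.
# Sprite and light renderers draw them, with color scaled by time of day, alpha fading in over system age and the blue glow flickering.
# The OrangeLight and TVLight emitters are disabled.
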NiagaraSystem /BRCosmetics/Weapons/FORT_Melee/Pickaxe_PawJasmineGravity/FX/NS_Pickaxe_PawJasmineGravity_Idle
dependencies {
 usedDataInterfaces [&NiagaraDataInterfaceCurve, &NiagaraDataInterfaceSocketReader]
 usedDynamicInputs []
 usedModules [
  &self:SystemSpawnScript.SystemScriptSource.SystemScriptGraph.NiagaraNodeAssignment_1.SetVariables_#1
  &self:SystemSpawnScript.SystemScriptSource.SystemScriptGraph.NiagaraNodeAssignment_4.SetVariables_#2
  &SystemState
  &self:SystemSpawnScript.SystemScriptSource.SystemScriptGraph.NiagaraNodeAssignment_0.SetVariables_#3
 ]
}
emitters {
 Glows {
  emitterUpdateScript [
   &EmitterState(
    Inactive Response=::NewEnumerator1
    Life Cycle Mode=::NewEnumerator1
    Loop Behavior=::NewEnumerator1
    Loop Duration Mode=::NewEnumerator1
    UseLoopDelay=false
   )
   &SpawnBurst_Instantaneous(Spawn Count=2, Use Spawn Probability=false, UseLoopCountLimit=false)
  ]
  particleSpawnScript [
   &PartitionParticles(Particles Per Partition=1, Partition Mode:ENiagara_PartitionMode_v2=::NewEnumerator1)
   &InitializeParticle(
    Color=&MergeHelper_ColorSwitch(
     0=<rgba 1 0.4778 0.033 0.075>
     1=<rgba 0.1268 0.4195 1 0.05>
     2=<rgba 0 0 0 0>
     3=<rgba 0 0 0 0>
     4=<rgba 0 0 0 0>
     Index=$Particles.PartitionParticles.Partition
     Num=::NewEnumerator3
    )
    Lifetime=5
    Sprite Size Mode=::NewEnumerator3
    Uniform Sprite Size=&MergeHelper_FloatSwitch(0=30, 1=50, 2=?, 3=?, Index=$Particles.PartitionParticles.Partition, Num=::NewEnumerator0)
    UsePositionOffset=false
    Write Material Random=true
   )
  ]
  particleUpdateScript [
   &ParticleState(Kill Particles When Lifetime Has Elapsed=false, Loop Particles Lifetime=true)
   &SocketLocation(
    SamplePosOffset=false
    Socket Index=$Particles.PartitionParticles.Partition
    Socket Reader=$System.Socket Reader
    Socket Sampling Mode=::NewEnumerator3
    VelocitySpeedLimit=false
   )
   &CameraOffset(Camera Offset Amount=&MergeHelper_FloatSwitch(
     0=&FN_FrontEndFloatModifier(Gameplay=15, Front End=10)
     1=5
     2=0
     3=0
     4=0
     Index=$Particles.PartitionParticles.Partition
     Num=::NewEnumerator3
    ))
   &FN_Light_Attributes(
    Color=$Particles.Color
    Create Color Attribute=true
    Create Radius Attribute=true
    Radius=&MergeHelper_FloatSwitch(0=20, 1=30, 2=0, 3=0, 4=0, Index=$Particles.PartitionParticles.Partition, Num=::NewEnumerator3)
    Scale Color=5
   )
   &ScaleColor(
    Scale Mode=::NewEnumerator1
    Scale RGBA=&MakeVector4_FromLinearColor(LinearColor=&MultiplyLinearColor(
      Linear Color=&FNTimeOfDayColor(Night Color=<rgba 1 1 1 1>, Morning Color=<rgba 1 0.8655 0.9596 1>, Day Color=<rgba 1 1 1 1>, Evening Color=<rgba 0.75 0.75 0.75 1>)
      Scale Factor=&MakeVector4FromVectorAndFloat(
       "Vector (RGB)"=<1 1 1>
       "Float (Alpha)"=&Multiply_Float(
        A=&MergeHelper_FloatSwitch(
         0=1
         1=&Waveform("Waveform [1]"=::NewEnumerator12, "[1] Frequency"=8, "[1] Amplitude Min/Max"=<0.75 1.1>)
         2=?
         3=?
         Index=$Particles.Initial.PartitionParticles.Partition
         Num=::NewEnumerator0
        )
        B=&FloatFromCurve(
         FloatCurve=DataInterface() {propertyValues {Curve {
            defaultValue=1
            keys [
             {arriveTangent=4.1785, interpMode=RCIM_Cubic, time=0, value=0}
             {arriveTangent=1, interpMode=RCIM_Cubic, leaveTangent=1, time=0.5, value=0.875}
             {interpMode=RCIM_Cubic, leaveTangent=0.0224, time=1, value=1}
            ]
           }}}
         CurveIndex=$System.Age
        )
       )
      )
     ))
    ScaleRGBA=true
   )
   &FN_Sprite_Attributes(
    Color=&MultiplyLinearColor(Linear Color=$Particles.Color, Scale Factor=<55 55 55 0.5>)
    Create Color Attribute=true
    Create Sprite Rotation Attribute=true
    Create SpriteSize Attribute=true
    Create Visibility Attribute=true
    SpriteRotation:NiagaraFloat=&Multiply_Float(A=$System.Age, B=10)
    SpriteSize=<15 15>
    Visibility Tag:NiagaraInt32=$Particles.PartitionParticles.Partition
    Write to Module Namespace=true
   )
  ]
  rendererClasses [&NiagaraSpriteRendererProperties, &NiagaraLightRendererProperties]
  renderers [
   {rendererClass=&NiagaraSpriteRendererProperties}
   {rendererClass=&NiagaraSpriteRendererProperties, rendererIndex=1}
   {rendererClass=&NiagaraLightRendererProperties, rendererIndex=2}
   {rendererClass=&NiagaraSpriteRendererProperties, rendererIndex=3}
  ]
 }
 OrangeLight {bEnabled=false, rendererClasses [], renderers []}
 TVLight {bEnabled=false, rendererClasses [], renderers []}
}
summary {
 systemName=NS_Pickaxe_PawJasmineGravity_Idle
 emitters [
  {bEnabled=false, emitterName=OrangeLight, rendererClasses []}
  {bEnabled=false, emitterName=TVLight, rendererClasses []}
  {emitterName=Glows, rendererClasses [&NiagaraSpriteRendererProperties, &NiagaraLightRendererProperties]}
 ]
 userVariables [
  User.bIsFrontEnd {type=NiagaraBool}
  User.bIsFrontEndPreview {type=NiagaraBool}
 ]
}
