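# Physics asset for the PawJasmineGravity pickaxe cosmetic, covering its dangling-style dyn_ bones: two cameras, a monitor and three two-segment cables.
# Each bone has a single capsule body, with a small capsule on b_AR_Root as the common anchor.
# Every dyn_ body is constrained directly to the root, so cable segments are not chained to each other.
# Narrow angular limits keep cable motion slight, while the cameras and monitor get wider, per-axis ranges.
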PhysicsAsset /BRCosmetics/Weapons/FORT_Melee/Pickaxe_PawJasmineGravity/Meshes/PawJasmineGravity_Axe_Physics
bodies {
 b_AR_Root [{shapeType=Capsule, length=1, radius=1}]
 dyn_cabel_01_01 [{shapeType=Capsule, length=10.2836, radius=2.2021, shapeName=dyn_cabel_01_01_capsule, center=<-3 0 0>, rotation=<rot 90 0 0>}]
 dyn_cabel_01_02 [{shapeType=Capsule, length=12.8794, radius=2.2856, shapeName=dyn_cabel_01_02_capsule, center=<-3 0 0>, rotation=<rot 90 0 0>}]
 dyn_cabel_02_01 [{shapeType=Capsule, length=7.7227, radius=0.6615, shapeName=dyn_cabel_02_01_capsule, center=<-3 0 0>, rotation=<rot 90 0 0>}]
 dyn_cabel_02_02 [{shapeType=Capsule, length=7.7775, radius=0.6268, shapeName=dyn_cabel_02_02_capsule, center=<-3 0 0>, rotation=<rot 90 0 0>}]
 dyn_cabel_03_01 [{shapeType=Capsule, length=11.5461, radius=1, shapeName=dyn_cabel_03_01_capsule, center=<-3 0 0>, rotation=<rot 90 0 0>}]
 dyn_cabel_03_02 [{
   shapeType=Capsule
   length=10.1817
   radius=1.549
   shapeName=dyn_cabel_03_02_capsule
   center=<-3 0.5748 -0.2141>
   rotation=<rot 0 111.9158 -87.3352>
  }]
 dyn_first_camera [{shapeType=Capsule, length=15, radius=3, shapeName=dyn_first_camera_capsule, center=<-9.8074 4.8231 0>, rotation=<rot 0.0005 8.8071 90.0001>}]
 dyn_monitor [{shapeType=Capsule, length=9.0877, radius=8.7678, shapeName=dyn_monitor_capsule, center=<-1 0 0>}]
 dyn_second_camera [{shapeType=Capsule, length=15, radius=4, shapeName=dyn_second_camera_capsule, center=<-9.4679 0 0>, rotation=<rot -6.8318 0.0001 0>}]
}
constraints [
 {bone1Name=dyn_first_camera, bone2Name=b_AR_Root, swing1LimitDegrees=35, swing2LimitDegrees=2, twistLimitDegrees=4}
 {bone1Name=dyn_monitor, bone2Name=b_AR_Root, swing1LimitDegrees=10, swing2LimitDegrees=5, twistLimitDegrees=15}
 {bone1Name=dyn_second_camera, bone2Name=b_AR_Root, swing1LimitDegrees=0, swing2LimitDegrees=15, twistLimitDegrees=25}
 {bone1Name=dyn_cabel_01_01, bone2Name=b_AR_Root, swing1LimitDegrees=5, swing2LimitDegrees=5, twistLimitDegrees=5}
 {bone1Name=dyn_cabel_01_02, bone2Name=b_AR_Root, swing1LimitDegrees=4, swing2LimitDegrees=4, twistLimitDegrees=4}
 {bone1Name=dyn_cabel_02_01, bone2Name=b_AR_Root, swing1LimitDegrees=3, swing2LimitDegrees=3, twistLimitDegrees=3}
 {bone1Name=dyn_cabel_02_02, bone2Name=b_AR_Root, swing1LimitDegrees=3, swing2LimitDegrees=3, twistLimitDegrees=3}
 {bone1Name=dyn_cabel_03_01, bone2Name=b_AR_Root, swing1LimitDegrees=3, swing2LimitDegrees=3, twistLimitDegrees=3}
 {bone1Name=dyn_cabel_03_02, bone2Name=b_AR_Root, swing1LimitDegrees=3, swing2LimitDegrees=3, twistLimitDegrees=3}
]
